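# Physics asset for the right pants piece of the TwinkleBot medium male player body, defining jointed capsules for its front and back bones.
# Small capsules on the front and back base bones act as parents for larger capsules on the matching dyn bones, which are offset along the bone.
# Each dyn bone is constrained only to its own base bone, with a narrow symmetric swing limit and zero twist limit.
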PhysicsAsset /BRCosmetics/Characters/Player/Male/Medium/Bodies/M_MED_TwinkleBot/Meshes/M_MED_TwinkleBot_Pants_R_Physics
bodies {
 base_pants_back_r [{shapeType=Capsule, length=0.505, radius=0.505, shapeName=base_pants_back_r_capsule, rotation=<rot 0 -90 0>}]
 base_pants_front_r [{shapeType=Capsule, length=0.505, radius=0.505, shapeName=base_pants_front_r_capsule, rotation=<rot 0 -90 0>}]
 dyn_pants_back_r [{shapeType=Capsule, length=2, radius=1, shapeName=dyn_pants_back_r_capsule, center=<5 0 0>, rotation=<rot 90 0 0>}]
 dyn_pants_front_r [{shapeType=Capsule, length=2, radius=1, shapeName=dyn_pants_front_r_capsule, center=<5 0 0>, rotation=<rot 90 0 0>}]
}
constraints [
 {bone1Name=dyn_pants_back_r, bone2Name=base_pants_back_r, swing1LimitDegrees=15, swing2LimitDegrees=15, twistLimitDegrees=0}
 {bone1Name=dyn_pants_front_r, bone2Name=base_pants_front_r, swing1LimitDegrees=15, swing2LimitDegrees=15, twistLimitDegrees=0}
]
